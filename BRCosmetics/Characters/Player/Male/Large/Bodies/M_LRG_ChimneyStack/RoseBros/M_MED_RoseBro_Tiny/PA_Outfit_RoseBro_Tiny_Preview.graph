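# Physics asset for the preview version of the RoseBro Tiny male outfit body.
# Nine single-capsule bodies cover the pelvis, upper spine, head, upper arms, hands and calves, with no separate thigh, forearm or foot bodies.
# Constraints form a tree rooted at the pelvis: calves and spine_05 attach to it, head and upper arms to spine_05, and each hand to its upper arm.
# All joints share uniform swing and twist limit angles.
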PhysicsAsset /BRCosmetics/Characters/Player/Male/Large/Bodies/M_LRG_ChimneyStack/RoseBros/M_MED_RoseBro_Tiny/PA_Outfit_RoseBro_Tiny_Preview
bodies {
 calf_l [{shapeType=Capsule, length=6.6721, radius=4.8042, center=<-4.3986 1.486 0.4588>, rotation=<rot 0 14.7269 -87.2876>}]
 calf_r [{shapeType=Capsule, length=6.6868, radius=4.8045, center=<4.398 -1.4811 -0.4348>, rotation=<rot 0 14.9097 -87.6397>}]
 hand_l [{shapeType=Capsule, length=5.9349, radius=4.2083, center=<3.7096 -0.5845 0.0096>, rotation=<rot 0 81.2107 -79.7172>}]
 hand_r [{shapeType=Capsule, length=5.9346, radius=4.2115, center=<-3.7093 0.5835 -0.0132>, rotation=<rot 0 81.2425 -79.6516>}]
 head [{shapeType=Capsule, length=21.8104, radius=16.7672, center=<9.784 2.2327 -1.1082>, rotation=<rot 4.6452 -90 -8.1541>}]
 pelvis [{shapeType=Capsule, length=14.3631, radius=12.496, center=<1.7332 -0.8766 -0.1136>, rotation=<rot 0.2003 -90 -89.4395>}]
 spine_05 [{shapeType=Capsule, length=11.9789, radius=7.076, center=<1.3959 1.3409 0.0001>, rotation=<rot -0.0507 -90 -0.0022>}]
 upperarm_l [{shapeType=Capsule, length=9.2243, radius=4.6433, center=<7.2402 1.0042 0.0124>, rotation=<rot 0 -73.9634 -87.8169>}]
 upperarm_r [{shapeType=Capsule, length=9.2281, radius=4.6446, center=<-7.2488 -1.008 -0.0113>, rotation=<rot 0 -73.9673 -87.8335>}]
}
constraints [
 {bone1Name=calf_l, bone2Name=pelvis, swing1LimitDegrees=45, swing2LimitDegrees=45, twistLimitDegrees=45}
 {bone1Name=calf_r, bone2Name=pelvis, swing1LimitDegrees=45, swing2LimitDegrees=45, twistLimitDegrees=45}
 {bone1Name=spine_05, bone2Name=pelvis, swing1LimitDegrees=45, swing2LimitDegrees=45, twistLimitDegrees=45}
 {bone1Name=head, bone2Name=spine_05, swing1LimitDegrees=45, swing2LimitDegrees=45, twistLimitDegrees=45}
 {bone1Name=upperarm_r, bone2Name=spine_05, swing1LimitDegrees=45, swing2LimitDegrees=45, twistLimitDegrees=45}
 {bone1Name=hand_r, bone2Name=upperarm_r, swing1LimitDegrees=45, swing2LimitDegrees=45, twistLimitDegrees=45}
 {bone1Name=upperarm_l, bone2Name=spine_05, swing1LimitDegrees=45, swing2LimitDegrees=45, twistLimitDegrees=45}
 {bone1Name=hand_l, bone2Name=upperarm_l, swing1LimitDegrees=45, swing2LimitDegrees=45, twistLimitDegrees=45}
]
