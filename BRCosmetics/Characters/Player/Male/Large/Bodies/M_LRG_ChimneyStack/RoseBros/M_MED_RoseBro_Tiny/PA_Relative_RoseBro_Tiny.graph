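# Physics asset for the RoseBro Tiny male player body cosmetic, giving its skeleton a simplified capsule-only set of physics bodies.
# Single capsules cover pelvis, two spine bones, neck, head, arms, hands, legs and feet, with a very large head and short, near-spherical limb shapes.
# Constraints chain these bodies out from the pelvis through spine and limbs.
# Elbow, neck and head joints explicitly set swing and twist to free.
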PhysicsAsset /BRCosmetics/Characters/Player/Male/Large/Bodies/M_LRG_ChimneyStack/RoseBros/M_MED_RoseBro_Tiny/PA_Relative_RoseBro_Tiny
bodies {
 calf_l [{shapeType=Capsule, length=0.0004, radius=4.3829, shapeName=calf_l_capsule, center=<-2.4051 0.0918 0.3998>, rotation=<rot -90 -0.0003 0>}]
 calf_r [{shapeType=Capsule, length=0.0004, radius=4.3829, shapeName=calf_r_capsule, center=<2.4051 -0.0918 -0.3998>, rotation=<rot -90 179.9997 0>}]
 foot_l [{shapeType=Capsule, length=6.6658, radius=3.8771, shapeName=foot_l_capsule, center=<0.588 2.358 0.9878>, rotation=<rot 0 0 96.7408>}]
 foot_r [{shapeType=Capsule, length=6.6658, radius=3.8771, shapeName=foot_r_capsule, center=<-0.588 -2.358 -0.9878>, rotation=<rot 0 0 -83.2592>}]
 hand_l [{shapeType=Capsule, length=3.1575, radius=4.6763, shapeName=hand_l_capsule, center=<5.3295 0.1601 -0.104>, rotation=<rot 90 -9.9998 0>}]
 hand_r [{shapeType=Capsule, length=3.1575, radius=4.6763, shapeName=hand_r_capsule, center=<-5.3295 -0.1601 0.104>, rotation=<rot 90 170.0002 0>}]
 head [{shapeType=Capsule, length=2.3908, radius=16.8052, shapeName=head_capsule, center=<9.3478 2.944 0>, rotation=<rot 0 49.9984 -90>}]
 lowerarm_l [{shapeType=Capsule, length=1.6457, radius=3.5102, shapeName=lowerarm_l_capsule, center=<4.2085 0.3091 0>, rotation=<rot 90 -179.9992 0>}]
 lowerarm_r [{shapeType=Capsule, length=1.6457, radius=3.5102, shapeName=lowerarm_r_capsule, center=<-4.2085 -0.3091 0>, rotation=<rot 90 0.0008 0>}]
 neck_02 [{shapeType=Capsule, length=4, radius=0.5092, shapeName=neck_02_capsule, center=<-1.6392 0.3806 0>, rotation=<rot -90 -0.0003 0>}]
 pelvis [{
   shapeType=Capsule
   length=8.5938
   radius=7.8453
   shapeName=pelvis_capsule
   center=<2.2891 -0.3031 -0.0012>
   rotation=<rot 0.0225 179.9992 180>
  }]
 spine_03 [{shapeType=Capsule, length=5.5975, radius=7.8135, shapeName=Spine_03_capsule, center=<1.078 0.5311 0>}]
 spine_05 [{shapeType=Capsule, length=6.5464, radius=6.1868, shapeName=spine_05_capsule, center=<0.2937 1.0208 0>, rotation=<rot 0 -0.0012 0>}]
 thigh_l [{shapeType=Capsule, length=0.5796, radius=4.0811, shapeName=thigh_l_capsule, center=<-2.6324 -0.0001 0.3374>, rotation=<rot -90 -0.0032 0>}]
 thigh_r [{shapeType=Capsule, length=0.5796, radius=4.0811, shapeName=thigh_r_capsule, center=<2.6324 0.0001 -0.3374>, rotation=<rot -90 179.9968 0>}]
 upperarm_l [{shapeType=Capsule, length=2.4721, radius=3.1838, shapeName=upperarm_l_capsule, center=<3.4346 -0.0125 0>, rotation=<rot 0.0001 86.8286 -90>}]
 upperarm_r [{shapeType=Capsule, length=2.4721, radius=3.1838, shapeName=upperarm_r_capsule, center=<-3.4346 0.0125 0>, rotation=<rot 0.0001 86.8286 90>}]
}
constraints [
 {bone1Name=spine_03, bone2Name=pelvis, swing1LimitDegrees=45, swing2LimitDegrees=45, twistLimitDegrees=45}
 {bone1Name=spine_05, bone2Name=spine_03, swing1LimitDegrees=45, swing2LimitDegrees=45, twistLimitDegrees=45}
 {bone1Name=upperarm_l, bone2Name=spine_05, swing1LimitDegrees=45, swing2LimitDegrees=45, twistLimitDegrees=45}
 {bone1Name=hand_l, bone2Name=lowerarm_l, swing1LimitDegrees=45, swing2LimitDegrees=45, twistLimitDegrees=45}
 {bone1Name=upperarm_r, bone2Name=spine_05, swing1LimitDegrees=45, swing2LimitDegrees=45, twistLimitDegrees=45}
 {bone1Name=hand_r, bone2Name=lowerarm_r, swing1LimitDegrees=45, swing2LimitDegrees=45, twistLimitDegrees=45}
 {bone1Name=thigh_l, bone2Name=pelvis, swing1LimitDegrees=45, swing2LimitDegrees=45, twistLimitDegrees=45}
 {bone1Name=calf_l, bone2Name=thigh_l, swing1LimitDegrees=45, swing2LimitDegrees=45, twistLimitDegrees=45}
 {bone1Name=foot_l, bone2Name=calf_l, swing1LimitDegrees=45, swing2LimitDegrees=45, twistLimitDegrees=45}
 {bone1Name=thigh_r, bone2Name=pelvis, swing1LimitDegrees=45, swing2LimitDegrees=45, twistLimitDegrees=45}
 {bone1Name=calf_r, bone2Name=thigh_r, swing1LimitDegrees=45, swing2LimitDegrees=45, twistLimitDegrees=45}
 {bone1Name=foot_r, bone2Name=calf_r, swing1LimitDegrees=45, swing2LimitDegrees=45, twistLimitDegrees=45}
 {
  bone1Name=lowerarm_r
  bone2Name=upperarm_r
  swing1LimitDegrees=45
  swing1Motion=Free
  swing2LimitDegrees=45
  swing2Motion=Free
  twistLimitDegrees=45
  twistMotion=Free
 }
 {
  bone1Name=lowerarm_l
  bone2Name=upperarm_l
  swing1LimitDegrees=45
  swing1Motion=Free
  swing2LimitDegrees=45
  swing2Motion=Free
  twistLimitDegrees=45
  twistMotion=Free
 }
 {
  bone1Name=neck_02
  bone2Name=spine_05
  swing1LimitDegrees=45
  swing1Motion=Free
  swing2LimitDegrees=45
  swing2Motion=Free
  twistLimitDegrees=45
  twistMotion=Free
 }
 {
  bone1Name=head
  bone2Name=neck_02
  swing1LimitDegrees=45
  swing1Motion=Free
  swing2LimitDegrees=45
  swing2Motion=Free
  twistLimitDegrees=45
  twistMotion=Free
 }
]
